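# M_Blank is a Default Lit placeholder material in the Hippie glider's folder, enabled for skeletal meshes, mesh deformers and particle sprites.
# Its graph holds a zero constant and the offset from world position back to the object position, made by negating their difference.
# Such an offset would collapse vertices onto the object's origin as World Position Offset, but no connection to a material output is shown.
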Material /Game/Gadgets/Assets/VinderTech_GliderChute/Glider_Hippie/Materials/M_Blank
graph {
 properties {
  bUsedWithMeshDeformer=true
  bUsedWithParticleSprites=true
  bUsedWithSkeletalMesh=true
  pixelDepthOffsetMode=PDOM_Legacy
  usedShadingModels=Default Lit
 }
 nodes [
  Constant()
  Multiply(A=Subtract(A=WorldPosition().XYZ, B=ObjectPositionWS())) {constB=-1}
 ]
}
